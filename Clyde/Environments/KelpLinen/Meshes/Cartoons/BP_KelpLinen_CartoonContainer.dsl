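; KelpLinen cartoon container actor that follows a socket on the KelpLinen mannequin's skeletal mesh and drives motion-reactive material effects, on non-server instances only.
; A throttled Tick moves the actor and samples hand and chest displacement, scaled by tick interval.
; Rolling arrays produce delayed, averaged velocities to smooth the motion.
; Hand velocity sets WPOVelocity on a dynamic smoke material on the hand glow mesh; chest velocity sets Cartoons_WPOVelocity in MPC_KelpLinen.
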(event EventBeginPlay
  (bind _kelplinenmannequin (Variables|Default|GetKelpLinenMannequin))
  (if (not (Networking|IsServer))
    ; Wait for the character to be valid
    (Utilities|IsValid _kelplinenmannequin
      (:Is%20Valid
        _)
      (:Is%20Not%20Valid
        (Utilities|FlowControl|Delay 1.0)))
    (Variables|Default|SetSKMReference (Actor|GetComponentbyClass _kelplinenmannequin "/Script/Engine.SkeletalMeshComponent"))
    (bind _returnvalue (Rendering|Material|CreateDynamicMaterialInstance (Variables|/Clyde/Environments/KelpLinen/BP/BP_KelpLinen_Mannequin.BP_KelpLinen_Mannequin_C|GetSM_FX_KelpLinen_HandGlowMesh_Frame1 (Variables|Default|GetKelpLinenMannequin)) 6 (Rendering|Material|GetMaterial (Variables|/Clyde/Environments/KelpLinen/BP/BP_KelpLinen_Mannequin.BP_KelpLinen_Mannequin_C|GetSM_FX_KelpLinen_HandGlowMesh_Frame1 (Variables|Default|GetKelpLinenMannequin)) 6) "None" :!class /Script/Engine.PrimitiveComponent))
    (Variables|Default|SetDynSmoke _returnvalue)
    (for _index (range 6 11)
      (Rendering|Material|SetMaterial (Variables|/Clyde/Environments/KelpLinen/BP/BP_KelpLinen_Mannequin.BP_KelpLinen_Mannequin_C|GetSM_FX_KelpLinen_HandGlowMesh_Frame1 (Variables|Default|GetKelpLinenMannequin)) _index (Variables|Default|GetDynSmoke)))))

(event EventTick
  (params (DeltaSeconds :type Float%20%28single-precision%29))
  ; Tick is slowed down to not run every frame, update position periodically. Spring arm handles interpolation
  (if (not (Networking|IsServer))
    (Utilities|IsValid (Variables|Default|GetSKMReference)
      (:Is%20Valid
        (Transformation|SetActorLocation (Transformation|GetSocketLocation (Variables|Default|GetSKMReference) (Variables|Default|GetSocketName)))
        ; Hand
        (Variables|Hand|SetHand_CurrentTickPos (Transformation|GetSocketLocation (Variables|Default|GetSKMReference) "hand_r"))
        (Variables|Hand|SetHand_CurentRealTimeVelocity (* (- (Variables|Hand|GetHand_PreviousTickPos) (Variables|Hand|GetHand_CurrentTickPos)) (Actor|Tick|GetActorTickInterval)))
        (Variables|Hand|SetHand_PreviousTickPos (Variables|Hand|GetHand_CurrentTickPos))
        (bind (_currentaverage _delayedaverage _hasenoughdata) (CallFunction|StoreNewHandVelocity (Variables|Hand|GetHand_CurentRealTimeVelocity)))
        (if _hasenoughdata
          (Rendering|Material|SetVectorParameterValue (Variables|Default|GetDynSmoke) "WPOVelocity" (Math|Color|MakeColor (.x _delayedaverage) (.y _delayedaverage) (.z _delayedaverage) (Variables|Default|GetMultiplier)) :!class /Script/Engine.MaterialInstanceDynamic)
          (else
            _))
        (Variables|Default|SetChest_CurrentTickPos (Transformation|GetSocketLocation (Variables|Default|GetSKMReference) (Variables|Default|GetSocketName)))
        (Variables|Default|SetChest_CurrentRealTimeVelocity (* (- (Variables|Default|GetChest_PreviousTickPos) (Variables|Default|GetChest_CurrentTickPos)) (Actor|Tick|GetActorTickInterval)))
        (Variables|Default|SetChest_PreviousTickPos (Variables|Default|GetChest_CurrentTickPos))
        (bind (_currentaverage_2 _delayedaverage_2) (CallFunction|StoreNewChestVelocity (Variables|Default|GetChest_CurrentRealTimeVelocity)))
        (Rendering|Material|SetVectorParameterValue "/ClydeCosmetics/Environments/KelpLinen/Materials/MPC_KelpLinen.MPC_KelpLinen" "Cartoons_WPOVelocity" (Math|Color|MakeColor (.x _delayedaverage_2) (.y _delayedaverage_2) (.z _delayedaverage_2) (Math|Vector|VectorLength _delayedaverage_2)) :!class /Script/Engine.KismetMaterialLibrary)))))

(event Custom|DebugUpdateLocation
  (bind _kelplinenmannequin (Variables|Default|GetKelpLinenMannequin))
  ; Debugging
  (Utilities|IsValid _kelplinenmannequin
    (:Is%20Valid
      (bind _output_get (Variables|Default|SetSKMReference (Actor|GetComponentbyClass _kelplinenmannequin "/Script/Engine.SkeletalMeshComponent")))
      (Utilities|IsValid _output_get
        (:Is%20Valid
          (Transformation|SetActorLocation (Transformation|GetSocketLocation (Variables|Default|GetSKMReference) (Variables|Default|GetSocketName))))))))

(fn StoreNewChestVelocity
  (params (In :type Vector))
  (Utilities|FlowControl|Sequence
    (:then_0
      ; Current Velocity
      (Utilities|Array|Insert (Variables|Default|GetChest_StoredVelocities) (Variables|GetIn))
      ;; Have we filled the array passed max
      (if (> (Utilities|Array|Length (Variables|Default|GetChest_StoredVelocities)) (Variables|Hand|GetHand_NumberOfVelocitiesToStore))
        (Utilities|Array|RemoveIndex (Variables|Default|GetChest_StoredVelocities) (Utilities|Array|LastIndex (Variables|Default|GetChest_StoredVelocities)))))
    (:then_1
      ; Delayed velocity
      (if (> (Utilities|Array|Length (Variables|Default|GetChest_StoredVelocities)) (Variables|Hand|GetHand_DelayedVelocityOffset))
        (Utilities|Array|Insert (Variables|Default|GetChest_DelayedStoredVelocities) (Utilities|Array|Get(acopy) (Variables|Default|GetChest_StoredVelocities) (Math|Integer|Clamp(Integer) (Variables|Hand|GetHand_DelayedVelocityOffset) 1 (Variables|Hand|GetHand_NumberOfVelocitiesToStore))))
        (if (> (Utilities|Array|Length (Variables|Default|GetChest_DelayedStoredVelocities)) (Variables|Hand|GetHand_NumberOfVelocitiesToStore))
          (Utilities|Array|RemoveIndex (Variables|Default|GetChest_DelayedStoredVelocities) (Utilities|Array|LastIndex (Variables|Default|GetChest_DelayedStoredVelocities))))))
    (:then_2
      (return (Math|Vector|GetVectorArrayAverage (Variables|Default|GetChest_StoredVelocities)) (Math|Vector|GetVectorArrayAverage (Variables|Default|GetChest_DelayedStoredVelocities))))))

(fn StoreNewHandVelocity
  (params (In :type Vector))
  (Utilities|FlowControl|Sequence
    (:then_0
      ; Current Velocity
      (Utilities|Array|Insert (Variables|Hand|GetHand_StoredVelocities) (Variables|GetIn))
      ;; Have we filled the array passed max
      (if (> (Utilities|Array|Length (Variables|Hand|GetHand_StoredVelocities)) (Variables|Hand|GetHand_NumberOfVelocitiesToStore))
        (Utilities|Array|RemoveIndex (Variables|Hand|GetHand_StoredVelocities) (Utilities|Array|LastIndex (Variables|Hand|GetHand_StoredVelocities)))))
    (:then_1
      ; Delayed velocity
      (if (> (Utilities|Array|Length (Variables|Hand|GetHand_StoredVelocities)) (Variables|Hand|GetHand_DelayedVelocityOffset))
        (Utilities|Array|Insert (Variables|Hand|GetHand_DelayedVelocityStorage) (Utilities|Array|Get(acopy) (Variables|Hand|GetHand_StoredVelocities) (Math|Integer|Clamp(Integer) (Variables|Hand|GetHand_DelayedVelocityOffset) 1 (Variables|Hand|GetHand_NumberOfVelocitiesToStore))))
        (if (> (Utilities|Array|Length (Variables|Hand|GetHand_DelayedVelocityStorage)) (Variables|Hand|GetHand_NumberOfVelocitiesToStore))
          (Utilities|Array|RemoveIndex (Variables|Hand|GetHand_DelayedVelocityStorage) (Utilities|Array|LastIndex (Variables|Hand|GetHand_DelayedVelocityStorage))))))
    (:then_2
      (return (Math|Vector|GetVectorArrayAverage (Variables|Hand|GetHand_StoredVelocities)) (Math|Vector|GetVectorArrayAverage (Variables|Hand|GetHand_DelayedVelocityStorage)) (or (> (Utilities|Array|Length (Variables|Hand|GetHand_DelayedVelocityStorage)) (Variables|Hand|GetHand_NumberOfVelocitiesToStore)) (== (Utilities|Array|Length (Variables|Hand|GetHand_DelayedVelocityStorage)) (Variables|Hand|GetHand_NumberOfVelocitiesToStore)))))))

(fn ConstructionScript)
