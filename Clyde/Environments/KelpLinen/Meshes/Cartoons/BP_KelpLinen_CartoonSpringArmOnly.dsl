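; Cartoon spring-arm actor for KelpLinen that follows a named socket on the KelpLinen mannequin's skeletal mesh, skipped on the server.
; BeginPlay caches the mannequin's skeletal mesh component, with a one-second delay if the mannequin is not yet valid.
; A throttled Tick snaps the actor to the socket location, leaving smoothing to spring arm interpolation.
; A debug event refreshes the mesh reference and position on demand, without the server check.
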(event EventBeginPlay
  (bind _kelplinenmannequin (Variables|Default|GetKelpLinenMannequin))
  (if (not (Networking|IsServer))
    ; Wait for the character to be valid
    (Utilities|IsValid _kelplinenmannequin
      (:Is%20Valid
        _)
      (:Is%20Not%20Valid
        (Utilities|FlowControl|Delay 1.0)))
    (Variables|Default|SetSKMReference (Actor|GetComponentbyClass _kelplinenmannequin "/Script/Engine.SkeletalMeshComponent"))))

(event EventTick
  (params (DeltaSeconds :type Float%20%28single-precision%29))
  ; Tick is slowed down to not run every frame, update position periodically. Spring arm handles interpolation
  (if (not (Networking|IsServer))
    (Utilities|IsValid (Variables|Default|GetSKMReference)
      (:Is%20Valid
        (Transformation|SetActorLocation (Transformation|GetSocketLocation (Variables|Default|GetSKMReference) (Variables|Default|GetSocketName)))))))

(event Custom|DebugUpdateLocation
  (bind _kelplinenmannequin (Variables|Default|GetKelpLinenMannequin))
  ; Debugging
  (Utilities|IsValid _kelplinenmannequin
    (:Is%20Valid
      (bind _output_get (Variables|Default|SetSKMReference (Actor|GetComponentbyClass _kelplinenmannequin "/Script/Engine.SkeletalMeshComponent")))
      (Utilities|IsValid _output_get
        (:Is%20Valid
          (Transformation|SetActorLocation (Transformation|GetSocketLocation (Variables|Default|GetSKMReference) (Variables|Default|GetSocketName))))))))

(fn ConstructionScript)
